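; Blueprint function library of shared helpers for Athena quest interactables.
; Queries check quest ownership, quest and objective completion (singly or ANDed across players), progress counts, actor tags and Creative/Playground playlist tags.
; Quest updates skip completed objectives, sending a row-handle stat event or falling back to the objective backend name.
; Feedback helpers play optional sound, Niagara and force feedback at player or quest actor; PartyAssistCheck only validates the controller.
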(fn AthenaHasQuest
  (params (InPlayerControlle :type Fort%20Player%20Controller%20Object%20Reference) (InQuestItemRef :type Fort%20Quest%20Item%20Definition%20Object%20Reference))
  ;[1 1 1] Newer method that handles both QS2 and QS1. It null checks the inputs
  (bind _returnvalue (Quest|GetQuestfromDefinition InPlayerControlle InQuestItemRef))
  (return (Utilities|IsValid _returnvalue :!class /Script/Engine.KismetSystemLibrary)))

(fn Athena_CheckActorHasTags
  (params (TargetActor :type Actor%20Object%20Reference) (TagsToCheck :type Gameplay%20Tag%20Container%20Structure) (OnlyExactMatch :type Boolean))
  (LocalVariables|SetbOnlyExactMatch OnlyExactMatch)
  (if (Utilities|IsValid TargetActor :!class /Script/Engine.KismetSystemLibrary)
    (LocalVariables|SetTagsWeWant TagsToCheck)
    (if (Utilities|DoesObjectImplementInterface TargetActor "/Script/GameplayTags.GameplayTagAssetInterface")
      (bind _returnvalue (GameplayTags|GetOwnedGameplayTags TargetActor :!class /Script/GameplayTags.GameplayTagAssetInterface))
      (return (GameplayTags|HasAllTags _returnvalue (LocalVariables|GetTagsWeWant) (LocalVariables|GetbOnlyExactMatch)))
      (else
        (return false)))
    (else
      (return false))))

(fn Athena_CheckCreativeMode
  (params (InGameState :type Game%20State%20Base%20Object%20Reference))
  (Class|FortKismetLibrary|ScriptEnsureIsValid "AthenaCheckCreativeMode" InGameState)
  (bind _asfort_game_state_athena (Utilities|Casting|CastToFortGameStateAthena InGameState))
  (if (or (GameplayTags|HasTag (Class|FortGameStateAthena|GetAthenaPlaylistContextTags _asfort_game_state_athena) "(TagName=\"Athena.Creative\")") (GameplayTags|HasTag (Class|FortGameStateAthena|GetAthenaPlaylistContextTags _asfort_game_state_athena) "(TagName=\"Athena.Playground\")"))
    (return false)
    (else
      (return true))))

(fn Athena_CheckQuestAndObjectiveCompletion
  (params (InPlayerControlle :type Fort%20Player%20Controller%20Object%20Reference) (InQuestItemRef :type Fort%20Quest%20Item%20Definition%20Object%20Reference) (ObjectiveBackendName :type Name))
  (Utilities|IsValid InPlayerControlle
    (:Is%20Valid
      (Utilities|IsValid InQuestItemRef
        (:Is%20Valid
          (bind _returnvalue (Quest|GetQuestfromDefinition InPlayerControlle InQuestItemRef))
          (if (Utilities|IsValid _returnvalue :!class /Script/Engine.KismetSystemLibrary)
            (if (Quest|HasCompletedQuest _returnvalue :!class /Script/FortniteGame.FortQuestItem)
              (return true true true)
              (elif (Quest|HasCompletedObjectivewithName _returnvalue ObjectiveBackendName :!class /Script/FortniteGame.FortQuestItem)
                (return true false true)
                (else
                  (return true false false))))
            (else
              _)))
        (:Is%20Not%20Valid
          _))
      (return false false false))
    (:Is%20Not%20Valid
      _)))

(fn Athena_CheckQuestAndObjectiveCompletionMP
  (params (InPlayerControlle :type Array%20of%20Fort%20Player%20Controller%20Object%20References) (InQuestItemRef :type Fort%20Quest%20Item%20Definition%20Object%20Reference) (ObjectiveBackendName :type Name))
  (for _array_element InPlayerControlle
    (bind (_questisvalid _questcompleted _objectivecompleted) (Default|AthenaCheckQuestandObjectiveCompletion _array_element InQuestItemRef ObjectiveBackendName (Variables|Get__WorldContext)))
    (Utilities|FlowControl|Sequence
      (:then_0
        (LocalVariables|SetResult_QuestIsValid (and _questisvalid (LocalVariables|GetResult_QuestIsValid))))
      (:then_1
        (LocalVariables|SetResult_QuestCompleted (and _questcompleted (LocalVariables|GetResult_QuestCompleted))))
      (:then_2
        (LocalVariables|SetResult_ObjectiveCompleted (and _objectivecompleted (LocalVariables|GetResult_ObjectiveCompleted))))))
  (return (LocalVariables|GetResult_QuestIsValid) (LocalVariables|GetResult_QuestCompleted) (LocalVariables|GetResult_ObjectiveCompleted)))

(fn Athena_GiveQuestUpdateToPlayers
  (params (InQuestItemRef :type Fort%20Quest%20Item%20Definition%20Object%20Reference) (ObjectiveBackendName :type Name) (ObjectiveStatEvent :type Data%20Table%20Row%20Handle%20Structure) (PlayerControllersForUpdate :type Array%20of%20Fort%20Player%20Controller%20Object%20References))
  (for _array_element PlayerControllersForUpdate
    (Development|PrintString 0 "Loopo")
    (Utilities|IsValid _array_element
      (:Is%20Valid
        (Development|PrintString 0 "valid")
        (Utilities|IsValid (Quests|GetQuestwithDefinition (Quests|GetQuestManager _array_element "Invalid") InQuestItemRef)
          (:Is%20Valid
            (if (not (Quest|HasCompletedObjectivewithName (Quests|GetQuestwithDefinition (Quests|GetQuestManager _array_element "Invalid") InQuestItemRef) ObjectiveBackendName :!class /Script/FortniteGame.FortQuestItem))
              (Utilities|IsValid (Utilities|Struct|BreakDataTableRowHandle ObjectiveStatEvent)
                (:Is%20Valid
                  ;[1 1 1] Feel free to delete if needed.
                  ;
                  ; 05/07/2020
                  ; Made this function now take rowhandles for legacy content, but will also work for BackendNames with the SendCustomStatEventDirect. Though usage of ComplexCustom is still likely the better answer.
                  ;
                  ; Custom stat is just the Enum for the StatEventType (Damage, Kill, Revive, Interact, etc)
                  ;
                  ; It is extremely strict on its usage of RowHandles. We'd like to move away from DataTable references and just do ComplexCustom, native types, and using the BackendName as it's much cleaner.
                  ;
                  ; ComplexCustom makes more sense for having events counting against multiple different quests, or re-use of a system.
                  ; -Nicholas.Lacelle
                  (Quest|SendCustomStatEvent (Quests|GetQuestManager _array_element "Invalid") ObjectiveStatEvent 1))
                (:Is%20Not%20Valid
                  (Quest|SendCustomStatEventDirect (Quests|GetQuestManager _array_element "Invalid") ObjectiveBackendName (Quests|GetQuestwithDefinition (Quests|GetQuestManager _array_element "Invalid") InQuestItemRef) 1)))))))
      (:Is%20Not%20Valid
        (Development|PrintString 0 "invalid")))))

(fn Athena_HasPlayerCompletedQuest
  (params (InPlayerControlle :type Fort%20Player%20Controller%20Object%20Reference) (InQuestItemRef :type Fort%20Quest%20Item%20Definition%20Object%20Reference))
  (bind _returnvalue (Quests|GetQuestManager InPlayerControlle "Invalid"))
  (bind _returnvalue_2 (Quests|GetQuestwithDefinition _returnvalue InQuestItemRef))
  (Utilities|IsValid InPlayerControlle
    (:Is%20Valid
      (Utilities|IsValid InQuestItemRef
        (:Is%20Valid
          (if (Utilities|IsValid _returnvalue_2 :!class /Script/Engine.KismetSystemLibrary)
            (if (Quest|HasCompletedQuest _returnvalue_2 :!class /Script/FortniteGame.FortQuestItem)
              (return true true)
              (else
                (return true false)))
            (else
              _)))
        (:Is%20Not%20Valid
          _))
      (return false false))
    (:Is%20Not%20Valid
      _)))

(fn Athena_IsCreativeOrPlaygroundPlaylist
  (params (TagContainer :type Gameplay%20Tag%20Container%20Structure%20%28by%20ref%29))
  (return (or (GameplayTags|HasTag TagContainer "(TagName=\"Athena.Creative\")") (GameplayTags|HasTag TagContainer "(TagName=\"Athena.Playground\")"))))

(fn Athena_PartyAssistCheck
  (params (PlayerController :type Fort%20Player%20Controller%20Object%20Reference) (ObjectiveBackendName :type Name))
  (Utilities|IsValid PlayerController
    (:Is%20Valid
      (return false true))
    (:Is%20Not%20Valid
      (return false false))))

(fn Athena_PlayQuestRewardFeedback
  (params (PlayFX_onPlayer? :type Boolean) (Success_Sound :type Sound%20Base%20Object%20Reference) (Success_Niagara :type Niagara%20System%20Object%20Reference) (Success_Shake :type Force%20Feedback%20Effect%20Object%20Reference) (QuestActor :type Actor%20Object%20Reference) (PlayerPawn :type Fort%20Player%20Pawn%20Object%20Reference) (Effect_Transform :type Transform))
  (bind _location (.location Effect_Transform))
  (bind _rotation (.rotation Effect_Transform))
  (bind _mesh (Variables|/Script/Engine.Character|GetMesh PlayerPawn))
  (Utilities|IsValid Success_Sound
    (:Is%20Valid
      (Audio|SpawnSoundAttached Success_Sound _mesh "head"))
    (:Is%20Not%20Valid
      _))
  (Utilities|IsValid Success_Niagara
    (:Is%20Valid
      (if PlayFX_onPlayer?
        (Niagara|SpawnSystemAttached Success_Niagara _mesh "None" _location _rotation "KeepRelativeOffset" true)
        (else
          (Niagara|SpawnSystematLocation 0 Success_Niagara (+ (Transformation|GetActorLocation QuestActor) _location) _rotation (.scale Effect_Transform)))))
    (:Is%20Not%20Valid
      _))
  (Utilities|IsValid Success_Shake
    (:Is%20Valid
      (Game|Feedback|ClientPlayForceFeedback (Pawn|GetLocalViewingPlayerController (Variables|GetPlayerPawn)) Success_Shake "None" false false false))))

(fn Athena_PlayQuestRewardFeedback_WithSoundAtQuestActorLocation
  (params (PlayFX_onPlayer? :type Boolean) (Success_Sound :type Sound%20Base%20Object%20Reference) (Success_Niagara :type Niagara%20System%20Object%20Reference) (Success_Shake :type Force%20Feedback%20Effect%20Object%20Reference) (QuestActor :type Actor%20Object%20Reference) (PlayerPawn :type Fort%20Player%20Pawn%20Object%20Reference) (Effect_Transform :type Transform))
  (bind _returnvalue (Transformation|GetActorLocation QuestActor))
  (bind _location (.location Effect_Transform))
  (bind _rotation (.rotation Effect_Transform))
  (bind _returnvalue_2 (+ _returnvalue _location))
  (Utilities|IsValid Success_Sound
    (:Is%20Valid
      (Audio|SpawnSoundatLocation QuestActor Success_Sound _returnvalue_2))
    (:Is%20Not%20Valid
      _))
  (Utilities|IsValid Success_Niagara
    (:Is%20Valid
      (if PlayFX_onPlayer?
        (Niagara|SpawnSystemAttached Success_Niagara (Variables|/Script/Engine.Character|GetMesh PlayerPawn) "None" _location _rotation "KeepRelativeOffset" true)
        (else
          (Niagara|SpawnSystematLocation 0 Success_Niagara _returnvalue_2 _rotation (.scale Effect_Transform)))))
    (:Is%20Not%20Valid
      _))
  (Utilities|IsValid Success_Shake
    (:Is%20Valid
      (Game|Feedback|ClientPlayForceFeedback (Pawn|GetLocalViewingPlayerController (Variables|GetPlayerPawn)) Success_Shake "None" false false false))))

(fn Athena_QuestObjectiveCounter
  (params (FortPlayerControllerAthena :type Fort%20Player%20Controller%20Object%20Reference) (ObjectiveItem :type Fort%20Quest%20Item%20Definition%20Object%20Reference) (ObjectiveStat :type Data%20Table%20Row%20Handle%20Structure))
  (bind _returnvalue (Quests|GetQuestManager FortPlayerControllerAthena "Invalid"))
  (bind _returnvalue_2 (Quests|GetQuestwithDefinition _returnvalue ObjectiveItem))
  (bind _returnvalue_3 (Quest|GetObjectiveInfo _returnvalue_2 ObjectiveStat))
  (Utilities|IsValid FortPlayerControllerAthena
    (:Is%20Valid
      (if (Utilities|IsValid _returnvalue_2 :!class /Script/Engine.KismetSystemLibrary)
        (LocalVariables|SetSuccessfullyFoundStage true)
        (if (Utilities|IsValid _returnvalue_3 :!class /Script/Engine.KismetSystemLibrary)
          (LocalVariables|SetLocalAchievedCount (Variables|/Script/FortniteGame.FortQuestObjectiveInfo|GetAchievedCount _returnvalue_3))
          (LocalVariables|SetLocalRequiredCount (Variables|/Script/FortniteGame.FortQuestObjectiveInfo|GetRequiredCount _returnvalue_3))
          (else
            _))
        (else
          (LocalVariables|SetSuccessfullyFoundStage false)))
      (return (LocalVariables|GetLocalAchievedCount) (LocalVariables|GetLocalRequiredCount) (LocalVariables|GetSuccessfullyFoundStage)))
    (:Is%20Not%20Valid
      _)))
