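# Unlit translucent particle-sprite material, named as a quest highlight for quest interactables, with depth test disabled so it draws through scene geometry.
# Particle color RGB feeds the MF_FX_EmissiveAdjustments function for the glow color.
# Particle alpha multiplies the alpha of a parameterized texture, defaulting to the HUD T_Zoom icon, to form a shaped, fadeable mask.
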Material /Game/Athena/Items/QuestInteractables/M_QuestHighlight
graph {
 properties {
  bDisableDepthTest=true
  bUsedWithParticleSprites=true
  blendMode=BLEND_Translucent
  pixelDepthOffsetMode=PDOM_Legacy
  shadingModel=MSM_Unlit
  usedShadingModels=Unlit
 }
 nodes [
  n1 = ParticleColor()
  MaterialFunctionCall(Emissive=@n1.RGB) {materialFunction=&MF_FX_EmissiveAdjustments}
  Multiply(A=@n1.A, B=TextureSample(Tex=TextureObjectParameter() {parameterName=TextureSample, texture=/Game/Athena/HUD/Art/T_Zoom}).A)
 ]
}
